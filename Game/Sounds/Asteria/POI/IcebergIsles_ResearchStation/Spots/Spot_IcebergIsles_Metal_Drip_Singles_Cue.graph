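# Sound cue for single metal drip spot ambience at the Iceberg Isles research station POI.
# A Random node picks from twelve equally weighted shipping container drip waves borrowed from the Artemis Hollow Base, and a Modulator varies each play.
# It uses a positional one-shot ambient attenuation asset and the outdoor ambient sound class, caps concurrency at two voices, and disables bus and submix sends.
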SoundCue /Game/Sounds/Asteria/POI/IcebergIsles_ResearchStation/Spots/Spot_IcebergIsles_Metal_Drip_Singles_Cue
root=Modulator_0
nodes [
 {id=Modulator_0, children [Random_1], properties {pitchMax=1.1}}
 {
  id=Random_1
  children [
   WavePlayer_42
   WavePlayer_43
   WavePlayer_44
   WavePlayer_45
   WavePlayer_46
   WavePlayer_47
   WavePlayer_48
   WavePlayer_49
   WavePlayer_50
   WavePlayer_51
   WavePlayer_52
   WavePlayer_53
  ]
  properties {weights [1,1,1,1,1,1,1,1,1,1,1,1]}
 }
 {id=WavePlayer_42, properties {soundWaveAssetPtr=/Game/Sounds/Artemis/POI/Hollow_Base/Spots/Spot_Shipping_Container_Interior_Drip_01}}
 {id=WavePlayer_43, properties {soundWaveAssetPtr=/Game/Sounds/Artemis/POI/Hollow_Base/Spots/Spot_Shipping_Container_Interior_Drip_02}}
 {id=WavePlayer_44, properties {soundWaveAssetPtr=/Game/Sounds/Artemis/POI/Hollow_Base/Spots/Spot_Shipping_Container_Interior_Drip_03}}
 {id=WavePlayer_45, properties {soundWaveAssetPtr=/Game/Sounds/Artemis/POI/Hollow_Base/Spots/Spot_Shipping_Container_Interior_Drip_04}}
 {id=WavePlayer_46, properties {soundWaveAssetPtr=/Game/Sounds/Artemis/POI/Hollow_Base/Spots/Spot_Shipping_Container_Interior_Drip_05}}
 {id=WavePlayer_47, properties {soundWaveAssetPtr=/Game/Sounds/Artemis/POI/Hollow_Base/Spots/Spot_Shipping_Container_Interior_Drip_06}}
 {id=WavePlayer_48, properties {soundWaveAssetPtr=/Game/Sounds/Artemis/POI/Hollow_Base/Spots/Spot_Shipping_Container_Interior_Drip_07}}
 {id=WavePlayer_49, properties {soundWaveAssetPtr=/Game/Sounds/Artemis/POI/Hollow_Base/Spots/Spot_Shipping_Container_Interior_Drip_08}}
 {id=WavePlayer_50, properties {soundWaveAssetPtr=/Game/Sounds/Artemis/POI/Hollow_Base/Spots/Spot_Shipping_Container_Interior_Drip_09}}
 {id=WavePlayer_51, properties {soundWaveAssetPtr=/Game/Sounds/Artemis/POI/Hollow_Base/Spots/Spot_Shipping_Container_Interior_Drip_10}}
 {id=WavePlayer_52, properties {soundWaveAssetPtr=/Game/Sounds/Artemis/POI/Hollow_Base/Spots/Spot_Shipping_Container_Interior_Drip_11}}
 {id=WavePlayer_53, properties {soundWaveAssetPtr=/Game/Sounds/Artemis/POI/Hollow_Base/Spots/Spot_Shipping_Container_Interior_Drip_12}}
]
properties {
 attenuationSettings=/Game/Sounds/Attenuation/Fort_TileAmbient_OneShot_MorePositional_Att
 bEnableBusSends=false
 bEnableSubmixSends=false
 bOverrideConcurrency=true
 maxDistance=8000
 soundClassObject=&Fort_Ambient_Outdoors_Athena
 volumeMultiplier=1.504
 attenuationOverrides {attenuationShape=Capsule, falloffDistance=200, stereoSpread=10000, attenuationShapeExtents=<236.449 200>}
 concurrencyOverrides {maxCount=2}
}
