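# Sound cue for a phone-ring ambient spot sound at the Iceberg Isles research station POI.
# A single wave player feeds the root, reusing an Apollo AccessHub phone ring wave with no randomization or looping set.
# It uses shared Athena ambient attenuation and the indoor ambient sound class at a heavily reduced volume.
# A send entry references the exterior ambient source bus, but its send level is zero.
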SoundCue /Game/Sounds/Asteria/POI/IcebergIsles_ResearchStation/Spots/Spot_IcebergIsles_PhoneRing_Cue
root=WavePlayer_0
nodes [{id=WavePlayer_0, properties {soundWaveAssetPtr=/Game/Sounds/Apollo/Biomes/AccessHub/Biome_AccessPoint_PhoneRing3_01}}]
properties {
 attenuationSettings=&Fort_Ambient_Athena_Att
 maxDistance=2097152
 soundClassObject=&Fort_Ambient_Indoors_Athena
 volumeMultiplier=0.025
 busSends [{audioBus=null, minSendLevel=0, sendLevel=0, soundSourceBus=&AmbientAudio_Exterior_SourceBus}]
}
