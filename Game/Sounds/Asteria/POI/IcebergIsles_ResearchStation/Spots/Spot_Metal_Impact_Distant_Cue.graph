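# Sound cue for distant metal impact spot ambience at the Iceberg Isles research station POI.
# A Random node picks among twelve equally weighted metal one-shots reused from the Apollo generic interior biome set.
# A Modulator then varies pitch downward and volume slightly per play.
# It outputs quietly through the indoor ambient sound class with a shared tile-ambient one-shot attenuation.
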SoundCue /Game/Sounds/Asteria/POI/IcebergIsles_ResearchStation/Spots/Spot_Metal_Impact_Distant_Cue
root=Modulator_0
nodes [
 {id=Modulator_0, children [Random_1], properties {pitchMax=1, pitchMin=0.8, volumeMax=1, volumeMin=0.7}}
 {
  id=Random_1
  children [
   WavePlayer_3
   WavePlayer_4
   WavePlayer_5
   WavePlayer_6
   WavePlayer_7
   WavePlayer_8
   WavePlayer_9
   WavePlayer_10
   WavePlayer_22
   WavePlayer_23
   WavePlayer_24
   WavePlayer_25
  ]
  properties {weights [1,1,1,1,1,1,1,1,1,1,1,1]}
 }
 {id=WavePlayer_10, properties {soundWaveAssetPtr=/Game/Sounds/Apollo/Biomes/Interiors_Generic/AMB_SS_Int_Metal_05}}
 {id=WavePlayer_22, properties {soundWaveAssetPtr=/Game/Sounds/Apollo/Biomes/Interiors_Generic/AMB_SS_Int_Metal_06}}
 {id=WavePlayer_23, properties {soundWaveAssetPtr=/Game/Sounds/Apollo/Biomes/Interiors_Generic/AMB_SS_Int_Metal_07}}
 {id=WavePlayer_24, properties {soundWaveAssetPtr=/Game/Sounds/Apollo/Biomes/Interiors_Generic/AMB_SS_Int_Metal_08}}
 {id=WavePlayer_25, properties {soundWaveAssetPtr=/Game/Sounds/Apollo/Biomes/Interiors_Generic/AMB_SS_Int_Metal_09}}
 {id=WavePlayer_3, properties {soundWaveAssetPtr=/Game/Sounds/Apollo/Biomes/Interiors_Generic/AMB_SS_Int_Metal_10}}
 {id=WavePlayer_4, properties {soundWaveAssetPtr=/Game/Sounds/Apollo/Biomes/Interiors_Generic/AMB_SS_Int_Metal_11}}
 {id=WavePlayer_5, properties {soundWaveAssetPtr=/Game/Sounds/Apollo/Biomes/Interiors_Generic/AMB_SS_Int_Metal_12}}
 {id=WavePlayer_6, properties {soundWaveAssetPtr=/Game/Sounds/Apollo/Biomes/Interiors_Generic/AMB_SS_Int_Metal_01}}
 {id=WavePlayer_7, properties {soundWaveAssetPtr=/Game/Sounds/Apollo/Biomes/Interiors_Generic/AMB_SS_Int_Metal_02}}
 {id=WavePlayer_8, properties {soundWaveAssetPtr=/Game/Sounds/Apollo/Biomes/Interiors_Generic/AMB_SS_Int_Metal_03}}
 {id=WavePlayer_9, properties {soundWaveAssetPtr=/Game/Sounds/Apollo/Biomes/Interiors_Generic/AMB_SS_Int_Metal_04}}
]
properties {
 attenuationSettings=/Game/Sounds/Attenuation/Fort_TileAmbient_OneShot_Att
 maxDistance=8000
 soundClassObject=&Fort_Ambient_Indoors_Athena
 volumeMultiplier=0.354
}
