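# SoundCue for a looping server rack hum spot sound in the Iceberg Isles research station office, built from reused computer transmitter loops.
# A Mixer blends the regular and low loops, weighted toward the regular one, with Envelopers fading each in at different rates and raising pitch.
# A quality switch feeds that mix to its first two slots and leaves two empty.
# Attenuation comes from a shared console-screen asset, and overridden concurrency stops the quietest instance.
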SoundCue /Game/Sounds/Asteria/POI/IcebergIsles_ResearchStation/Spots/Spot_Office_ServerRack_Hum_Loop_Cue
root=QualityLevel_0
nodes [
 {
  id=Enveloper_2
  children [WavePlayer_3]
  properties {pitchCurve {editorCurveData {keys [<at 0.2394 1.2173>]}}, volumeCurve {editorCurveData {keys [<at 0.0233 0.0116>,<at 0.7456 0.9951>]}}}
 }
 {
  id=Enveloper_5
  children [WavePlayer_8]
  properties {pitchCurve {editorCurveData {keys [<at 0.2392 1.2207>]}}, volumeCurve {editorCurveData {keys [<at 0.0233 0.0116>,<at 1.5065 0.9996>]}}}
 }
 {id=Mixer_0, children [Enveloper_2,Enveloper_5], properties {inputVolume [1.2,0.8]}}
 {id=QualityLevel_0, children [Mixer_0,Mixer_0,null,null]}
 {
  id=WavePlayer_3
  properties {bLooping=true, soundWaveAssetPtr=/Game/Athena/Sounds/Props/MountainBaseComputer/ComputerTransimtter_Running_Loop_01}
 }
 {
  id=WavePlayer_8
  properties {bLooping=true, soundWaveAssetPtr=/Game/Athena/Sounds/Props/MountainBaseComputer/ComputerTransimtter_Running_Loop_Low_01}
 }
]
properties {
 attenuationSettings=/Game/Sounds/Artemis/Environment_Actors/OilRigControlConsole/Spot_Console_Screen_Loop_Att
 bOverrideConcurrency=true
 maxDistance=650
 priority=50
 soundClassObject=&Fort_Gadgets
 volumeMultiplier=1
 attenuationOverrides {
  bEnableFocusInterpolation=true
  bEnableListenerFocus=true
  distanceAlgorithm=Logarithmic
  falloffDistance=1100
  focusAttackInterpSpeed=20
  focusAzimuth=20
  focusReleaseInterpSpeed=10
  focusVolumeAttenuation=1.75
  lPFFrequencyAtMax=10000
  lPFRadiusMax=1000
  lPFRadiusMin=500
  nonFocusAzimuth=40
  nonFocusVolumeAttenuation=0.8
  occlusionInterpolationTime=0.3
  occlusionLowPassFilterFrequency=3000
  occlusionVolumeAttenuation=0.75
  stereoSpread=10000
  attenuationShapeExtents {x=0}
  customAttenuationCurve {editorCurveData {keys [<at 0 1 RCIM_Cubic>, {arriveTangent=-1, interpMode=RCIM_Cubic, leaveTangent=-1, time=0.5, value=0.1}, <at 1 0 RCIM_Cubic>]}}
 }
 concurrencyOverrides {resolutionRule=StopQuietest}
}
