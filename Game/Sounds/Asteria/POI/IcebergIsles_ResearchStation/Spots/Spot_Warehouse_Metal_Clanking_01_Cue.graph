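# Spatialized one-shot ambient cue for metal clanking in the warehouse of the Iceberg Isles research station.
# The first two quality levels mix randomly picked creaks, delayed distant impacts and quieter light industrial hits, with pitch and volume modulation.
# Empty random slots sometimes silence a layer, and random delays stagger the creaks and impacts.
# The last two quality levels instead play one of three modulated single waves, a cheaper variant.
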SoundCue /Game/Sounds/Asteria/POI/IcebergIsles_ResearchStation/Spots/Spot_Warehouse_Metal_Clanking_01_Cue
root=QualityLevel_0
nodes [
 {id=Delay_12, children [Modulator_36], properties {delayMax=3}}
 {id=Delay_13, children [Modulator_37], properties {delayMax=2, delayMin=1}}
 {id=Mixer_19, children [Delay_12,Delay_13,Modulator_38], properties {inputVolume [2,1,0.5]}}
 {id=Modulator_1, children [WavePlayer_19], properties {pitchMax=1.1, pitchMin=0.6, volumeMax=1.5, volumeMin=0.6}}
 {id=Modulator_2, children [WavePlayer_18], properties {pitchMax=1.1, pitchMin=0.5, volumeMax=1, volumeMin=0.3}}
 {id=Modulator_3, children [WavePlayer_17], properties {pitchMax=1, pitchMin=0.5, volumeMax=0.5, volumeMin=0.2}}
 {id=Modulator_35, children [Mixer_19], properties {pitchMax=1.1, pitchMin=0.9, volumeMax=1.1, volumeMin=0.9}}
 {id=Modulator_36, children [Random_0], properties {pitchMax=1.1, pitchMin=0.9, volumeMax=1, volumeMin=0.3}}
 {id=Modulator_37, children [Random_2], properties {pitchMax=1.1, pitchMin=0.9, volumeMax=1, volumeMin=0.3}}
 {id=Modulator_38, children [Random_3], properties {pitchMax=1, pitchMin=0.5, volumeMax=1, volumeMin=0.3}}
 {id=QualityLevel_0, children [Modulator_35,Modulator_35,Random_4,Random_4]}
 {
  id=Random_0
  children [WavePlayer_14, WavePlayer_12, WavePlayer_13, WavePlayer_15, null]
  properties {bRandomizeWithoutReplacement=false, weights [1,1,1,1,0.5]}
 }
 {
  id=Random_2
  children [WavePlayer_0, WavePlayer_1, WavePlayer_2, WavePlayer_11, null]
  properties {bRandomizeWithoutReplacement=false, weights [1,1,1,1,0.5]}
 }
 {
  id=Random_3
  children [WavePlayer_16, WavePlayer_29, WavePlayer_31, null]
  properties {bRandomizeWithoutReplacement=false, weights [1,1,1,1]}
 }
 {id=Random_4, children [Modulator_1,Modulator_2,Modulator_3], properties {weights [1,1,1]}}
 {id=WavePlayer_0, properties {soundWaveAssetPtr=/Game/Sounds/Fort_Ambient_NonLoop/Industrial/Industrial_Distant_Impact01}}
 {id=WavePlayer_1, properties {soundWaveAssetPtr=/Game/Sounds/Fort_Ambient_NonLoop/Industrial/Industrial_Distant_Impact02}}
 {id=WavePlayer_11, properties {soundWaveAssetPtr=/Game/Sounds/Fort_Ambient_NonLoop/Industrial/Industrial_Distant_Impact04}}
 {id=WavePlayer_12, properties {soundWaveAssetPtr=/Game/Sounds/Fort_Ambient_NonLoop/Industrial/Industrial_Distant_Creak02}}
 {id=WavePlayer_13, properties {soundWaveAssetPtr=/Game/Sounds/Fort_Ambient_NonLoop/Industrial/Industrial_Distant_Creak03}}
 {id=WavePlayer_14, properties {soundWaveAssetPtr=/Game/Sounds/Fort_Ambient_NonLoop/Industrial/Industrial_Distant_Creak01}}
 {id=WavePlayer_15, properties {soundWaveAssetPtr=/Game/Sounds/Fort_Ambient_NonLoop/Industrial/Industrial_Distant_Creak04}}
 {id=WavePlayer_16, properties {soundWaveAssetPtr=/Game/Sounds/Fort_Ambient_NonLoop/Industrial/fort_industrial_light_05a}}
 {id=WavePlayer_17, properties {soundWaveAssetPtr=&fort_industrial_light_05a}}
 {id=WavePlayer_18, properties {soundWaveAssetPtr=&Industrial_Distant_Impact02}}
 {id=WavePlayer_19, properties {soundWaveAssetPtr=&Industrial_Distant_Creak01}}
 {id=WavePlayer_2, properties {soundWaveAssetPtr=/Game/Sounds/Fort_Ambient_NonLoop/Industrial/Industrial_Distant_Impact03}}
 {id=WavePlayer_29, properties {soundWaveAssetPtr=/Game/Sounds/Fort_Ambient_NonLoop/Industrial/fort_industrial_light_06a}}
 {id=WavePlayer_31, properties {soundWaveAssetPtr=/Game/Sounds/Fort_Ambient_NonLoop/Industrial/fort_industrial_light_07a}}
]
properties {
 attenuationSettings=/Game/Sounds/Attenuation/Fort_TileAmbient_OneShot_Att
 maxDistance=8000
 soundClassObject=&Fort_Ambient_Indoors_Athena
 volumeMultiplier=0.501
}
