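# Ambient one-shot SoundCue of distant metal clanking for the Iceberg Isles research station warehouse, using indoor ambient sound class and one-shot attenuation.
# The first two quality slots mix delayed creak, delayed impact and undelayed light industrial layers, each randomly picking from four waves or an empty slot.
# The last two slots play just one random wave from a three-wave subset.
# Modulators randomize pitch and volume throughout for variation.
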SoundCue /Game/Sounds/Asteria/POI/IcebergIsles_ResearchStation/Spots/Spot_Warehouse_Metal_Clanking_02_Cue
root=QualityLevel_0
nodes [
 {id=Delay_12, children [Modulator_1], properties {delayMax=3}}
 {id=Delay_13, children [Modulator_4], properties {delayMax=4, delayMin=3}}
 {id=Mixer_19, children [Delay_12,Delay_13,Modulator_7], properties {inputVolume [2,1,0.5]}}
 {id=Modulator_1, children [Random_0], properties {pitchMax=1.1, pitchMin=0.9, volumeMax=1, volumeMin=0.3}}
 {id=Modulator_12, children [WavePlayer_13], properties {pitchMax=1.1, pitchMin=0.6, volumeMax=1.5, volumeMin=0.6}}
 {id=Modulator_13, children [WavePlayer_14], properties {pitchMax=1.1, pitchMin=0.5, volumeMax=1, volumeMin=0.3}}
 {id=Modulator_14, children [WavePlayer_18], properties {pitchMax=1, pitchMin=0.5, volumeMax=0.5, volumeMin=0.2}}
 {id=Modulator_35, children [Mixer_19], properties {pitchMax=1.1, pitchMin=0.9, volumeMax=1.1, volumeMin=0.9}}
 {id=Modulator_4, children [Random_2], properties {pitchMax=1.1, pitchMin=0.9, volumeMax=1, volumeMin=0.3}}
 {id=Modulator_7, children [Random_3], properties {pitchMax=1, pitchMin=0.5, volumeMax=1, volumeMin=0.3}}
 {id=QualityLevel_0, children [Modulator_35,Modulator_35,Random_5,Random_5]}
 {
  id=Random_0
  children [WavePlayer_11, WavePlayer_15, WavePlayer_16, WavePlayer_17, null]
  properties {bRandomizeWithoutReplacement=false, weights [1,1,1,1,1]}
 }
 {
  id=Random_2
  children [WavePlayer_0, WavePlayer_1, WavePlayer_2, WavePlayer_26, null]
  properties {bRandomizeWithoutReplacement=false, weights [1,1,1,1,1]}
 }
 {
  id=Random_3
  children [WavePlayer_12, WavePlayer_33, WavePlayer_34, WavePlayer_35, null]
  properties {bRandomizeWithoutReplacement=false, weights [1,1,1,1,1]}
 }
 {id=Random_5, children [Modulator_12,Modulator_13,Modulator_14], properties {weights [1,1,1]}}
 {id=WavePlayer_0, properties {soundWaveAssetPtr=/Game/Sounds/Fort_Ambient_NonLoop/Industrial/Industrial_Distant_Impact09}}
 {id=WavePlayer_1, properties {soundWaveAssetPtr=/Game/Sounds/Fort_Ambient_NonLoop/Industrial/Industrial_Distant_Impact10}}
 {id=WavePlayer_11, properties {soundWaveAssetPtr=/Game/Sounds/Fort_Ambient_NonLoop/Industrial/Industrial_Distant_Creak05}}
 {id=WavePlayer_12, properties {soundWaveAssetPtr=/Game/Sounds/Fort_Ambient_NonLoop/Industrial/fort_industrial_light_08a}}
 {id=WavePlayer_13, properties {soundWaveAssetPtr=&Industrial_Distant_Creak05}}
 {id=WavePlayer_14, properties {soundWaveAssetPtr=&Industrial_Distant_Impact09}}
 {id=WavePlayer_15, properties {soundWaveAssetPtr=/Game/Sounds/Fort_Ambient_NonLoop/Industrial/Industrial_Distant_Creak06}}
 {id=WavePlayer_16, properties {soundWaveAssetPtr=/Game/Sounds/Fort_Ambient_NonLoop/Industrial/Industrial_Distant_Creak07}}
 {id=WavePlayer_17, properties {soundWaveAssetPtr=/Game/Sounds/Fort_Ambient_NonLoop/Industrial/Industrial_Distant_Creak08}}
 {id=WavePlayer_18, properties {soundWaveAssetPtr=&fort_industrial_light_08a}}
 {id=WavePlayer_2, properties {soundWaveAssetPtr=/Game/Sounds/Fort_Ambient_NonLoop/Industrial/Industrial_Distant_Impact11}}
 {id=WavePlayer_26, properties {soundWaveAssetPtr=/Game/Sounds/Fort_Ambient_NonLoop/Industrial/Industrial_Distant_Impact12}}
 {id=WavePlayer_33, properties {soundWaveAssetPtr=/Game/Sounds/Fort_Ambient_NonLoop/Industrial/fort_industrial_light_09a}}
 {id=WavePlayer_34, properties {soundWaveAssetPtr=/Game/Sounds/Fort_Ambient_NonLoop/Industrial/fort_industrial_light_10a}}
 {id=WavePlayer_35, properties {soundWaveAssetPtr=/Game/Sounds/Fort_Ambient_NonLoop/Industrial/fort_industrial_light_11a}}
]
properties {
 attenuationSettings=/Game/Sounds/Attenuation/Fort_TileAmbient_OneShot_Att
 maxDistance=8000
 soundClassObject=&Fort_Ambient_Indoors_Athena
 volumeMultiplier=0.501
}
